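; Wall-mounted indoor fan prop Blueprint that assembles a motor and blades onto the inherited static mesh at construction time.
; UpdateFanParts chains parts by socket, moving the motor to the base mesh's first socket and the blades to the motor's first socket, if one exists.
; After the parent construction script and assembly, blades and motor are colored white when no building texture data exists.
; The event graph holds only empty events.
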(event EventBeginPlay)

(event Collision|EventActorBeginOverlap
  (params (OtherActor :type Actor%20Object%20Reference)))

(event EventTick
  (params (DeltaSeconds :type Float%20%28single-precision%29)))

(fn UpdateFanParts
  (bind _sm_appliance_indoorfanmotor_a (Variables|Default|GetSM_Appliance_IndoorFanMotor_A))
  (bind _returnvalue (Transformation|GetAllSocketNames _sm_appliance_indoorfanmotor_a))
  (Transformation|SetWorldLocation (Variables|Default|GetSM_Appliance_IndoorFanMotor_A) (Transformation|GetSocketLocation (Variables|Debug|GetStaticMeshComponent) (Utilities|Array|Get(acopy) (Transformation|GetAllSocketNames (Variables|Debug|GetStaticMeshComponent)))) false false)
  (if (Utilities|Array|IsValidIndex _returnvalue)
    (Transformation|SetWorldLocation (Variables|Default|GetSM_Appliance_IndoorFanBlades_A) (Transformation|GetSocketLocation _sm_appliance_indoorfanmotor_a (Utilities|Array|Get(acopy) _returnvalue)) false false)))

(fn ConstructionScript
  (|CallParentFunction)
  (CallFunction|UpdateFanParts)
  (bind _outtexturedataarray (Building|GetBuildingTextureData))
  (if (not (Utilities|Array|IsValidIndex _outtexturedataarray))
    (Utilities|FlowControl|Sequence
      (:then_0
        (Default|SetColoronSecondaryMesh "(R=1.000000,G=1.000000,B=1.000000,A=1.000000)" (Variables|Default|GetSM_Appliance_IndoorFanBlades_A)))
      (:then_1
        (Default|SetColoronSecondaryMesh "(R=1.000000,G=1.000000,B=1.000000,A=1.000000)" (Variables|Default|GetSM_Appliance_IndoorFanMotor_A))))))
